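# Translucent, two-sided trail material for the CyberKarate pickaxe, set up for Niagara ribbons and beam trails.
# A lightning texture scrolls with UVs warped by two panning noise samples, then gets shaped by an inverted sphere mask.
# An edge glow from rounded U and a sphere mask is added, and the sum fades across V.
# Particle color tints it for emissive, while particle alpha and a second shape mask derive opacity from it.
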Material /BRCosmetics/Weapons/FORT_Melee/Pickaxe_CyberKarate/FX/Materials/M_Trail_CyberKarate
graph {
 properties {
  bUsedWithBeamTrails=true
  bUsedWithNiagaraRibbons=true
  blendMode=BLEND_Translucent
  pixelDepthOffsetMode=PDOM_Legacy
  translucencyLightingMode=TLM_VolumetricPerVertexDirectional
  twoSided=true
  usedShadingModels=Default Lit
 }
 nodes [
  n1 = ParticleColor()
  n2 = TextureCoordinate() {uTiling=1.25, vTiling=0.5}
  n3 = Add(A=ComponentMask(None=@n2) {r=true}) {constB=-0.125}
  n4 = AppendVector(A=Constant() {r=0.7}, B=Constant() {r=0.5})
  n5 = Constant() {r=0.9}
  n6 = Constant() {r=0.6476}
  n7 = OneMinus(None=Max(A=Round(None=@n3), B=SphereMask(A=@n2, B=@n4, Radius=@n5, Hardness=@n6)))
  n8 = ScalarParameter() {defaultValue=2, parameterName=EdgeGlow Int}
  n9 = SphereMask(
   A=TextureCoordinate() {uTiling=0.7}
   B=AppendVector(A=ScalarParameter() {defaultValue=0.8, parameterName=SphereX}, B=ScalarParameter() {defaultValue=0.8, parameterName=SphereY})
   Radius=Constant() {r=0.85}
   Hardness=Constant() {r=0.6}
  )
  n10 = Multiply(
   A=MaterialFunctionCall(XY=TextureCoordinate()) {materialFunction=&Swizzle}.YX
   B=AppendVector(A=ScalarParameter() {defaultValue=1, parameterName=VTile}, B=ScalarParameter() {defaultValue=0.1, parameterName=UTile})
  )
  n11 = AppendVector(A=ScalarParameter() {defaultValue=0.666, parameterName=SpeedY}, B=ScalarParameter() {defaultValue=-0.75, parameterName=SpeedX})
  n12 = TextureObject() {texture=&T_NoisePack_01}
  n13 = AppendVector(
   A=TextureSample(UVs=Panner() {speedX=-0.3913, speedY=0.0981}, Tex=@n12) {texture=&T_NoisePack_01}.R
   B=TextureSample(UVs=Panner() {speedX=0.4865, speedY=-0.255}, Tex=@n12) {texture=&T_NoisePack_01}.G
  )
  n14 = ScalarParameter() {defaultValue=0.1, parameterName=WarpInt}
  n15 = TextureSample(UVs=Add(A=Panner(Coordinate=@n10, Speed=@n11), B=Multiply(A=@n13, B=@n14))) {texture=/Game/Weapons/WeaponSkins/Textures/T_Lightning_Drift_v2}
  n16 = Power(Base=Subtract(A=@n15.RGB) {constB=0.01}) {constExponent=0.5}
  n17 = Multiply(A=Power(Base=OneMinus(None=@n9)) {constExponent=5}, B=Subtract(A=Saturate(None=@n16), B=@n15.RGB))
  n18 = ScalarParameter() {defaultValue=12, parameterName=ShapeInt}
  n19 = Constant() {r=0.5}
  n20 = MaterialFunctionCall(Float2=TextureCoordinate()) {materialFunction=&BreakOutFloat2Components}
  n21 = Multiply(
   A=Add(A=Multiply(A=@n7, B=@n8), B=Multiply(A=@n17, B=@n18))
   B=SphereMask(
    A=AppendVector(A=@n19, B=@n20.G)
    B=@n19
    Radius=ScalarParameter() {defaultValue=0.5, parameterName=Mask Radius}
    Hardness=ScalarParameter() {defaultValue=0.75, parameterName=Mask Hardness}
   )
  )
  MaterialFunctionCall(Emissive=Multiply(A=@n1.RGB, B=@n21)) {materialFunction=&MF_FX_EmissiveAdjustments}
  n22 = TextureCoordinate() {uTiling=1.25}
  n23 = Add(A=ComponentMask(None=@n22) {r=true}) {constB=-0.125}
  n24 = AppendVector(A=Constant() {r=0.7}, B=Constant() {r=0.5})
  n25 = Constant() {r=0.7}
  n26 = Constant() {r=0.6476}
  Multiply(A=Max(A=Round(None=@n23), B=SphereMask(A=@n22, B=@n24, Radius=@n25, Hardness=@n26)), B=Multiply(A=@n1.A, B=@n21))
 ]
}
