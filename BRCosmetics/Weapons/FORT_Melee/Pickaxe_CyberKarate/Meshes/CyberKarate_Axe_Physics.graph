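# Physics asset for the CyberKarate two-handed pickaxe mesh, defining collision bodies and joints for its dangling wires.
# A capsule covers the handle, and a compound of boxes and a capsule covers the dyn_arm bone.
# Four three-segment capsule wire chains (a–d) each hang from dyn_arm, with every link limited to moderate swing and tight twist.
# No constraint joins dyn_arm to the handle.
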PhysicsAsset /BRCosmetics/Weapons/FORT_Melee/Pickaxe_CyberKarate/Meshes/CyberKarate_Axe_Physics
bodies {
 Melee_TwoHanded_Handle [{shapeType=Capsule, length=96.7696, radius=2.5224, center=<0 -0.843 41.6934>}]
 dyn_arm [
  {shapeType=Capsule, length=13.8782, radius=0.85, center=<-1.5264 -13.912 -1.0956>, rotation=<rot 0 0 90>}
  {shapeType=Box, extentX=12.6636, extentY=12.7751, extentZ=3.8623, center=<3.4337 -12.9216 -7.5094>}
  {shapeType=Box, extentX=12.6636, extentY=12.7751, extentZ=3.8623, center=<-4.8749 -11.891 -4.5337>, rotation=<rot -44.9998 0 0>}
  {shapeType=Box, extentX=12.6636, extentY=12.7751, extentZ=3.8623, center=<-7.2814 -11.891 3.7888>, rotation=<rot -90 -0.0297 0.0297>}
  {shapeType=Box, extentX=4.5, extentY=7.5, extentZ=4.5, center=<2.1624 -14.2882 0.9734>}
 ]
 dyn_wire_a01 [{shapeType=Capsule, length=4.5691, radius=1.0354, center=<2.8926 0 0>, rotation=<rot 90 0 0>}]
 dyn_wire_a02 [{shapeType=Capsule, length=4.5691, radius=1.0354, center=<2.8358 0.0001 0>, rotation=<rot 90 0 0>}]
 dyn_wire_a03 [{shapeType=Capsule, length=3.9646, radius=1.0354, center=<2 0 0>, rotation=<rot 90 0 0>}]
 dyn_wire_b01 [{shapeType=Capsule, length=3.2527, radius=1.0354, center=<2 0 0>, rotation=<rot 90 0 0>}]
 dyn_wire_b02 [{shapeType=Capsule, length=5.0117, radius=1.0354, center=<2.8882 0 0>, rotation=<rot 90 0 0>}]
 dyn_wire_b03 [{shapeType=Capsule, length=2.8933, radius=1.0354, center=<1.4945 0 0>, rotation=<rot 90 0 0>}]
 dyn_wire_c01 [{shapeType=Capsule, length=3.9646, radius=1.0354, center=<2 0 0>, rotation=<rot 90 0 0>}]
 dyn_wire_c02 [{shapeType=Capsule, length=4.4148, radius=1.0354, center=<3.0946 0 0>, rotation=<rot 90 0 0>}]
 dyn_wire_c03 [{shapeType=Capsule, length=3.9646, radius=1.0354, center=<2 0 0>, rotation=<rot 90 0 0>}]
 dyn_wire_d01 [{shapeType=Capsule, length=3.4565, radius=1.0354, center=<2 0 0>, rotation=<rot 90 0 0>}]
 dyn_wire_d02 [{shapeType=Capsule, length=4.4674, radius=1.0354, center=<2.8064 0 0>, rotation=<rot 90 0 0>}]
 dyn_wire_d03 [{shapeType=Capsule, length=3.9646, radius=1.0354, center=<2 0 0>, rotation=<rot 90 0 0>}]
}
constraints [
 {bone1Name=dyn_wire_a01, bone2Name=dyn_arm, swing1LimitDegrees=30, swing2LimitDegrees=30, twistLimitDegrees=5}
 {bone1Name=dyn_wire_a02, bone2Name=dyn_wire_a01, swing1LimitDegrees=30, swing2LimitDegrees=30, twistLimitDegrees=5}
 {bone1Name=dyn_wire_a03, bone2Name=dyn_wire_a02, swing1LimitDegrees=30, swing2LimitDegrees=30, twistLimitDegrees=5}
 {bone1Name=dyn_wire_b01, bone2Name=dyn_arm, swing1LimitDegrees=30, swing2LimitDegrees=30, twistLimitDegrees=5}
 {bone1Name=dyn_wire_b02, bone2Name=dyn_wire_b01, swing1LimitDegrees=30, swing2LimitDegrees=30, twistLimitDegrees=5}
 {bone1Name=dyn_wire_b03, bone2Name=dyn_wire_b02, swing1LimitDegrees=30, swing2LimitDegrees=30, twistLimitDegrees=5}
 {bone1Name=dyn_wire_c01, bone2Name=dyn_arm, swing1LimitDegrees=30, swing2LimitDegrees=30, twistLimitDegrees=5}
 {bone1Name=dyn_wire_c02, bone2Name=dyn_wire_c01, swing1LimitDegrees=30, swing2LimitDegrees=30, twistLimitDegrees=5}
 {bone1Name=dyn_wire_c03, bone2Name=dyn_wire_c02, swing1LimitDegrees=30, swing2LimitDegrees=30, twistLimitDegrees=5}
 {bone1Name=dyn_wire_d01, bone2Name=dyn_arm, swing1LimitDegrees=30, swing2LimitDegrees=30, twistLimitDegrees=5}
 {bone1Name=dyn_wire_d02, bone2Name=dyn_wire_d01, swing1LimitDegrees=30, swing2LimitDegrees=30, twistLimitDegrees=5}
 {bone1Name=dyn_wire_d03, bone2Name=dyn_wire_d02, swing1LimitDegrees=30, swing2LimitDegrees=30, twistLimitDegrees=5}
]
